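# Translucent unlit mesh-particle material for the Sleepy Time backpack's FX, drawing light lines or, with the For_stars switch, stars.
# Time-of-day color times particle color feeds emissive adjustments.
# Line opacity combines panning cloud, smoke and fire noise, curve-atlas edge falloffs and depth fade; star opacity uses textures, swirl noise and view-facing fade, skipping both fades on mobile.
# World position offset uses SplineThicken for lines, noise-driven normal displacement for stars.
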Material /BRCosmetics/Accessories/FORT_Backpacks/M_MED_Sleepy_Time_Backpack/FX/M_SleepyTime_Backpack_LightLine
graph {
 properties {
  bUsedWithMeshParticles=true
  bUsedWithNiagaraMeshParticles=true
  blendMode=BLEND_Translucent
  pixelDepthOffsetMode=PDOM_Legacy
  shadingModel=MSM_Unlit
  twoSided=true
  usedShadingModels=Unlit
 }
 nodes [
  n1 = VectorParameter() {parameterName=DayColor, defaultValue=<rgba 1.5 1.0157 0.3359 1>}
  n2 = VectorParameter() {parameterName=NightColor, defaultValue=<rgba 2 1.2285 0.1458 1>}
  n3 = ParticleColor()
  MaterialFunctionCall(Emissive=Multiply(
    A=MaterialFunctionCall(Color_Morning=@n1.RGB, Color_Day=@n1.RGB, Color_Evening=@n1.RGB, Color_Night=@n2.RGB) {materialFunction=&Color_Time_of_Day}.Result
    B=@n3.RGB
   )) {materialFunction=&MF_FX_EmissiveAdjustments}
  Multiply()
  n4 = DotProduct(A=CameraVectorWS(), B=VertexNormalWS())
  n5 = TextureSample() {texture=/Game/Accessories/FORT_Backpacks/M_MED_Sleepy_Time_Backpack/FX/T_SleepyTime_Elements}
  n6 = Multiply(A=Time()) {constB=0.03}
  n7 = VertexColor()
  n8 = Multiply(A=Time()) {constB=0.01}
  n9 = TextureSample(UVs=AppendVector(A=Add(A=@n6, B=@n7.R), B=Add(A=@n8, B=@n7.R))) {samplerType=SAMPLERTYPE_LinearGrayscale, texture=/Game/Global/Textures/Default/Noise/T_SwirlNoise_N_height}
  n10 = MaterialFunctionCall(In=@n9.R, Contrast=ScalarParameter() {defaultValue=2, parameterName=SurfaceStarBrightContrast}) {materialFunction=&CheapContrast}
  n11 = TextureSample() {texture=&T_SleepyTime_Elements}
  n12 = LinearInterpolate(A=ScalarParameter() {defaultValue=-1, parameterName=Param_3}, B=ScalarParameter() {defaultValue=2, parameterName=Param_2}, Alpha=@n9.R)
  n13 = Multiply(A=LinearInterpolate(A=@n11.B, B=@n11.B, Alpha=Clamp(None=@n12)), B=@n9.R)
  n14 = VectorParameter() {parameterName=Noise3_texture, defaultValue=<rgba 0.4 0.4 -0.1 0.05>}
  n15 = AppendVector(A=@n14.B, B=@n14.B)
  n16 = TextureSample(UVs=Multiply(A=Panner(Speed=@n15), B=AppendVector(A=@n14.R, B=@n14.R))) {texture=&AtmosphericCloudNoise01}
  n17 = Multiply(
   A=ScalarParameter() {defaultValue=1, parameterName=StarOpacity}
   B=StaticSwitchParameter(True=Multiply(A=@n5.R, B=@n10.Result), False=Multiply(A=@n13, B=@n16.R)) {parameterName=ForSurfaceStarOnly}
  )
  n18 = Multiply(
   A=MaterialFunctionCall(In=Abs(None=@n4), Contrast=Constant() {r=1}) {materialFunction=&CheapContrast}.Result
   B=@n17
  )
  n19 = VertexColor()
  n20 = LinearInterpolate(Alpha=@n19.R) {constA=0.3, constB=1.5}
  n21 = VectorParameter() {parameterName=Noise1_texture, defaultValue=<rgba 0.08 0.08 -0.1 0.05>}
  n22 = AppendVector(A=@n21.B, B=@n21.B)
  n23 = Multiply(A=Panner(Speed=Multiply(A=@n20, B=@n22)), B=AppendVector(A=@n21.R, B=@n21.R))
  n24 = Add(A=@n20, B=@n23)
  n25 = LinearInterpolate(Alpha=@n19.R) {constA=0.3, constB=1.5}
  n26 = VectorParameter() {parameterName=Noise2_texture, defaultValue=<rgba 0.1 0.1 0.2 0.1>}
  n27 = AppendVector(A=@n26.B, B=@n26.B)
  n28 = Multiply(A=Panner(Speed=Multiply(A=@n25, B=@n27)), B=AppendVector(A=@n26.R, B=@n26.R))
  n29 = TextureSample(UVs=Add(A=@n25, B=@n28)) {texture=/Game/Global/Textures/Default/Tileable_Patterns/T_Smoke_Puff_Tile_Up}
  n30 = CurveAtlasRowParameter(CurveTime=ComponentMask(None=TextureCoordinate()) {g=true}) {
   atlas=&AthenaCurveAtlas01
   curve=/Game/Accessories/FORT_Backpacks/M_MED_Sleepy_Time_Backpack/FX/Curve_SleepyTime
   defaultValue=127
  }
  n31 = ScalarParameter() {parameterName=LightTopThickness}
  n32 = OneMinus(None=@n31)
  n33 = TextureCoordinate()
  n34 = ComponentMask(None=@n33) {r=true}
  n35 = ScalarParameter() {defaultValue=-0.7, parameterName=LightBottomThickness}
  n36 = OneMinus(None=@n35)
  n37 = CurveAtlasRowParameter(CurveTime=LinearInterpolate(A=LinearInterpolate(A=@n31, B=@n32, Alpha=@n34), B=LinearInterpolate(A=@n35, B=@n36, Alpha=@n34), Alpha=ComponentMask(None=@n33) {g=true})) {
   atlas=&AthenaCurveAtlas01
   curve=&Curve_SleepyTime
   defaultValue=127
  }
  n38 = Constant() {r=0.3}
  n39 = OneMinus(None=@n38)
  n40 = TextureCoordinate()
  n41 = ComponentMask(None=@n40) {r=true}
  n42 = Constant() {r=-0.7}
  n43 = OneMinus(None=@n42)
  n44 = CurveAtlasRowParameter(CurveTime=ComponentMask(None=@n40) {g=true}) {
   atlas=&AthenaCurveAtlas01
   curve=&Curve_SleepyTime
   defaultValue=127
  }
  n45 = CurveAtlasRowParameter(CurveTime=LinearInterpolate(A=LinearInterpolate(A=@n38, B=@n39, Alpha=@n41), B=LinearInterpolate(A=@n42, B=@n43, Alpha=@n41), Alpha=@n44.A)) {
   atlas=&AthenaCurveAtlas01
   curve=&Curve_SleepyTime
   defaultValue=127
  }
  n46 = TextureCoordinate()
  n47 = Power(Base=ComponentMask(None=@n46) {g=true}) {constExponent=1.8}
  n48 = Sine(None=Multiply(A=ComponentMask(None=@n46) {r=true}) {constB=0.5})
  n49 = Add(A=Multiply(A=@n30.R, B=@n37.G), B=Multiply(A=Multiply(A=@n45.B, B=@n47), B=Power(Base=@n48))) {constB=0}
  n50 = Multiply(A=Multiply(A=TextureSample(UVs=@n24) {texture=&AtmosphericCloudNoise01}.R, B=Power(Base=@n29.R)), B=@n49)
  n51 = ScalarParameter() {defaultValue=0.2, parameterName=LineThickness}
  n52 = OneMinus(None=@n51)
  n53 = TextureCoordinate()
  n54 = ComponentMask(None=@n53) {r=true}
  n55 = ScalarParameter() {defaultValue=-1, parameterName=LineBottomThickness}
  n56 = OneMinus(None=@n55)
  n57 = ComponentMask(None=@n53) {g=true}
  n58 = AppendVector(A=LinearInterpolate(A=LinearInterpolate(A=@n51, B=@n52, Alpha=@n54), B=LinearInterpolate(A=@n55, B=@n56, Alpha=@n54), Alpha=@n57), B=@n57)
  n59 = VectorParameter() {parameterName=Line_texture, defaultValue=<rgba 0.5 0.03 0.1 0>}
  n60 = AppendVector(A=@n59.B, B=@n59.B)
  n61 = Add(
   A=Multiply(A=Panner(Coordinate=@n58, Speed=@n60), B=ComponentMask(None=@n59.RGB) {g=true, r=true})
   B=AppendVector(A=VertexColor().R, B=Constant() {r=1})
  )
  n62 = Clamp(None=LinearInterpolate(
    A=ScalarParameter() {defaultValue=0.45, parameterName=Param_4}
    B=ScalarParameter() {defaultValue=1, parameterName=Param_5}
    Alpha=TextureSample(UVs=@n61) {texture=/Game/Global/Textures/Default/Noise/T_FireNoise_01}.R
   ) {constA=0.5, constB=2})
  n63 = Multiply(
   A=LinearInterpolate(
    A=ScalarParameter() {defaultValue=-0.1, parameterName=Dark}
    B=ScalarParameter() {defaultValue=10, parameterName=Boost}
    Alpha=Multiply(A=@n50, B=@n62)
   )
   B=MaterialFunctionCall(In=@n9.R, Contrast=ScalarParameter() {defaultValue=10, parameterName=Param_6}) {materialFunction=&CheapContrast}.Result
  )
  n64 = Clamp(None=@n63)
  n65 = Multiply(A=@n49) {constB=0.1}
  n66 = Multiply(
   A=ScalarParameter() {defaultValue=0.08, parameterName=Opacity}
   B=StaticSwitchParameter(True=Add(A=@n64, B=@n65) {constB=0}, False=@n63) {parameterName=LineAlwaysLit}
  )
  n67 = DepthFade(Opacity=@n66, FadeDistance=ScalarParameter() {defaultValue=2, parameterName=DepthFade}) {fadeDistanceDefault=2}
  Multiply(
   A=@n3.A
   B=StaticSwitchParameter(
    True=ShadingPathSwitch(Default=@n18, Deferred=@n18, Mobile=@n17) {desc=Migrated from FeatureLevelSwitch}
    False=ShadingPathSwitch(Default=@n67, Deferred=@n67, Mobile=@n66) {desc=Migrated from FeatureLevelSwitch}
   ) {parameterName=For_stars}
  )
  Reroute(None=@n23)
  Reroute(None=@n28)
  n68 = VertexNormalWS()
  n69 = ScalarParameter() {defaultValue=5, parameterName=Normal_intensity}
  n70 = ScalarParameter() {defaultValue=-2.5, parameterName=Param_8}
  n71 = ScalarParameter() {defaultValue=0.8, parameterName=Param_7}
  n72 = ScalarParameter() {defaultValue=25, parameterName=WidthBase}
  StaticSwitchParameter(
   True=Multiply(A=Multiply(A=@n68, B=@n69), B=LinearInterpolate(A=@n70, B=@n71, Alpha=@n9.R) {constA=-2, constB=0.8})
   False=MaterialFunctionCall(WidthBase=@n72, WidthTip=@n72) {materialFunction=&SplineThicken}.WorldPositionOffset
  ) {parameterName=For_stars}
 ]
}
