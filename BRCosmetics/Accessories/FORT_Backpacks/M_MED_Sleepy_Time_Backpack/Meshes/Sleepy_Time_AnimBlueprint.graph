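# Animation Blueprint for the Sleepy Time back bling, built to spin its dyn_ball bone and react to the wearer's kills.
# AnimGraph adds ball_rotator to the bone in bone space, with a wrapping Spin_Int counter supplying the roll.
# A Do Once leads to an async AthenaKills stat watch that flags Active_Kill on increases.
# That flag steers Spin_Delta interpolation, but its branch has no incoming exec and Spin_Delta never feeds the rotation.
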AnimBlueprint /BRCosmetics/Accessories/FORT_Backpacks/M_MED_Sleepy_Time_Backpack/Meshes/Sleepy_Time_AnimBlueprint
target_skeleton=&/Sleepy_Time_Skeleton
graphs [
 AnimGraph {nodes [
   n1 = "Transform (Modify) Bone - Bone: dyn_ball"(
    ComponentPose=Local To Component(LocalPose=Local Space Ref Pose())
    Alpha=1
    Translation="0,0,0"
    Rotation=Get ball_rotator() -> (ball_rotator:Rotator)
    Scale="1,1,1"
   ) {rotationMode=BMM_Additive, rotationSpace=BCS_BoneSpace, alphaBoolBlend {blendOption=Linear}, boneToModify {boneName=dyn_ball}}
   Output Pose(Result=Component To Local(ComponentPose=@n1)) {layerGroup=None}
  ]}
 EventGraph {nodes [
   n1 = Cast To FortPlayerPawn(execute=@n9.Completed, Object=TryGetFortPawn())
   n2 = WatchCosmeticStat!(
    execute:Exec
    Pawn:Fort Player Pawn Object Reference=@n1
    StatTag:Fort Stat Manager Tag Structure=(TagName="StatManager.AthenaKills")
    DefaultValue:int=0
    bBroadcastInitialValue:bool=true
    InitialValueBroadcastDelay:float=-1
    ChangeValueBroadcastDelay:float=-1
    bConsumeWaitingNotifications:bool=true
   ) -> (then:Exec, OnStatChanged:Exec, StatName:Name, StatValue:int) {is=title K2Node_AsyncAction}
   Branch(
    execute=@n2.OnStatChanged
    Condition="float > float"(A="To Float (Integer)"(InInt=@n2.StatValue), B=Get KillCount() -> (KillCount:double)) {is=title K2Node_CallFunction}
   )
   > Set Active_Kill(Active_Kill:bool=true) -> (Output_Get:bool)
   > Set KillCount(KillCount:double=float + float(A=Get KillCount(), B=1) {is=title K2Node_CommutativeAssociativeBinaryOperator}) -> (Output_Get:double)
   Event BlueprintUpdateAnimation()
   n3 = Get Spin_Int() -> (Spin_Int:double)
   > Set Spin_Int(Spin_Int:double=float + float(A=@n3, B=1)) -> (Output_Get:double)
   > n4 = Branch(Condition="float >= float"(A=@n3, B=720) {is=title K2Node_CallFunction})
   > n5 = Set Spin_Int(Spin_Int=float - float(A=Get Spin_Int(), B=720) {is=title K2Node_CallFunction})
   n6 = Set Spin_Int(execute=@n4.else, Spin_Int=float + float(A=Get Spin_Int(), B=1))
   Set ball_rotator(execute=[@n5.then, @n6.then], ball_rotator:Rotator=MakeRotator(Roll=float * float(A=@n6, B=-1))) -> (Output_Get:Rotator)
   n7 = Branch(Condition=Get Active_Kill() -> (Active_Kill:bool))
   > n8 = Set Spin_Delta(Spin_Delta:double=FInterpTo(Current=Get Spin_Delta() -> (Spin_Delta:double), Target=25, DeltaTime=0, InterpSpeed=4)) -> (Output_Get:double)
   > Branch(Condition="float >= float"(A=@n8, B=25))
   > Set Active_Kill(Active_Kill=false)
   Set Spin_Delta(execute=@n7.else, Spin_Delta=FInterpTo(Current=Get Spin_Delta(), Target=1, DeltaTime=0, InterpSpeed=1))
   Reroute Node()
   > n9 = Do Once()
  ]}
]
variables [
 Active_Kill {type=boolean}
 ball_rotator {type=Rotator}
 KillCount {type=number}
 Spin_Delta {type=number}
 Spin_Int {type=number}
]
